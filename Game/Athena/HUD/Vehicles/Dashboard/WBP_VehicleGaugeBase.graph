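# Base widget Blueprint for vehicle dashboard gauges, deriving from FortAthenaVehicleDashboardWidget with no visual tree, that drives fuel and disabled warnings.
# On vehicle update, it caches whether the vehicle is electric and starts async waits on the Vehicle.Disabled tag to toggle the disabled warning.
# Fuel handlers show or clear flag-tracked warnings, choosing electric or non-electric variants.
# GetWarningWidget returns nothing here, apparently leaving child Blueprints to supply the warning widget.
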WidgetBlueprint /Game/Athena/HUD/Vehicles/Dashboard/WBP_VehicleGaugeBase
graphs [
 EventGraph {nodes [
   n1 = Event OnVehicleUpdated!() -> (OutputDelegate, then, NewVehicle:Fort Athena Vehicle Object Reference) {is=title K2Node_Event}
   n2 = Get DisabledTag() -> (DisabledTag:Gameplay Tag Structure)
   > n3 = WaitGameplayTagAddToActor!(execute:Exec, TargetActor:Actor Object Reference=@n1.NewVehicle, Tag:Gameplay Tag Structure=@n2, OnlyTriggerOnce:bool=false) -> (then:Exec, AsyncTaskProxy:Ability Async Wait Gameplay Tag Added Object Reference, Added:Exec) {is=title K2Node_AsyncAction}
   > n4 = WaitGameplayTagRemoveFromActor!(execute:Exec, TargetActor:Actor Object Reference=@n1.NewVehicle, Tag:Gameplay Tag Structure=@n2, OnlyTriggerOnce:bool=false) -> (then:Exec, AsyncTaskProxy:Ability Async Wait Gameplay Tag Removed Object Reference, Removed:Exec) {is=title K2Node_AsyncAction}
   > Set bIsVehicleElectric(bIsVehicleElectric:bool=HasMatchingGameplayTag!(self:Gameplay Tag Asset Interface Interface=@n1.NewVehicle, TagToCheck:Gameplay Tag Structure=(TagName="Vehicle.Fuel.Electric")) -> (ReturnValue:bool)) -> (Output_Get:bool)
   OnVehicleEnabled!(execute=@n4.Removed, self, IsEnabled:bool=true) -> (then)
   OnVehicleEnabled(execute=@n3.Added, IsEnabled=false)
  ]}
 GetWarningWidget {nodes [
   GetWarningWidget() {is=title K2Node_FunctionEntry}
   > Return Node!(execute, WarningWidget:WBP Vehicle Warning Object Reference)
  ]}
 OnLowFuel {nodes [
   OnLowFuel() {is=title K2Node_FunctionEntry}
   > n1 = Set WarningWidget(WarningWidget:WBP Vehicle Warning Object Reference=GetWarningWidget!(self) -> (WarningWidget:WBP Vehicle Warning Object Reference) {is=title K2Node_CallFunction}) -> (Output_Get:WBP Vehicle Warning Object Reference)
   > Is Valid(InputObject=@n1)
   > n2 = Branch(Condition=Get OutOfFuelWarningActive() -> (OutOfFuelWarningActive:bool))
   > RemoveWarning!(
    execute
    self:WBP Vehicle Warning Object Reference=Get WarningWidget() -> (WarningWidget:WBP Vehicle Warning Object Reference)
    Type:ENUM_WarningType Enum=Select(
     Option 0:ENUM_WarningType Enum=NewEnumerator1
     Option 1:ENUM_WarningType Enum=NewEnumerator7
     Index=Get bIsVehicleElectric() -> (bIsVehicleElectric:bool)
    ) -> (ReturnValue:ENUM_WarningType Enum)
   ) -> (then)
   > n3 = Set OutOfFuelWarningActive(OutOfFuelWarningActive:bool=false) -> (Output_Get:bool)
   Branch(execute=[@n3.then, @n2.else])
   > ShowWarning!(
    execute
    self:WBP Vehicle Warning Object Reference=Get WarningWidget()
    Type:ENUM_WarningType Enum=Select(Option 0=NewEnumerator0, Option 1=NewEnumerator6, Index=Get bIsVehicleElectric())
   ) -> (then)
   > Set LowFuelWarningActive(LowFuelWarningActive:bool=true) -> (Output_Get:bool)
  ]}
 OnOutOfFuel {nodes [
   OnOutOfFuel() {is=title K2Node_FunctionEntry}
   > n1 = Set WarningWidget(WarningWidget=GetWarningWidget())
   > Is Valid(InputObject=@n1)
   > ShowWarning(self=Get WarningWidget(), Type=Select(Option 0=NewEnumerator1, Option 1=NewEnumerator7, Index=Get bIsVehicleElectric()))
   > Set OutOfFuelWarningActive(OutOfFuelWarningActive=true)
  ]}
 OnRegularFuel {nodes [
   OnRegularFuel() {is=title K2Node_FunctionEntry}
   > n1 = Set WarningWidget(WarningWidget=GetWarningWidget())
   > Is Valid(InputObject=@n1)
   > n2 = Sequence()
   > Branch(Condition=Get OutOfFuelWarningActive())
   > RemoveWarning(self=Get WarningWidget(), Type=Select(Option 0=NewEnumerator1, Option 1=NewEnumerator7, Index=Get bIsVehicleElectric()))
   > Set OutOfFuelWarningActive(OutOfFuelWarningActive=false)
   Branch(execute=@n2.then_1, Condition=Get LowFuelWarningActive() -> (LowFuelWarningActive:bool))
   > RemoveWarning(self=Get WarningWidget(), Type=Select(Option 0=NewEnumerator0, Option 1=NewEnumerator6, Index=Get bIsVehicleElectric()))
   > Set LowFuelWarningActive(LowFuelWarningActive=false)
  ]}
 OnTankFull {nodes [
   OnTankFull() {is=title K2Node_FunctionEntry}
   > Parent: OnTankFull() {is=title K2Node_CallParentFunction}
   > RemoveRefuelingWarning()
  ]}
 OnVehicleEnabled {nodes [
   n1 = OnVehicleEnabled!() -> (then, IsEnabled:bool=~) {is=title K2Node_FunctionEntry}
   > n2 = Set WarningWidget(WarningWidget=GetWarningWidget())
   > Is Valid(InputObject=@n2)
   > n3 = Branch(Condition=@n1)
   n4 = Get WarningWidget()
   > RemoveWarning(self=@n4, Type=NewEnumerator2)
   ShowWarning(execute=@n3.else, self=@n4, Type=NewEnumerator2)
  ]}
 RemoveRefuelingWarning {nodes [
   RemoveRefuelingWarning() {is=title K2Node_FunctionEntry}
   > Branch(Condition=Get RefuelingWarningActive() -> (RefuelingWarningActive:bool))
   n1 = GetWarningWidget()
   > Is Valid(InputObject=@n1)
   > RemoveWarning(self=@n1, Type=Select(Option 0=NewEnumerator5, Option 1=NewEnumerator4, Index=Get bIsVehicleElectric()))
   > Set RefuelingWarningActive(RefuelingWarningActive:bool=false) -> (Output_Get:bool)
  ]}
 SetVehicleHasInfiniteFuel {nodes [
   n1 = SetVehicleHasInfiniteFuel!() -> (then, bInfiniteFuel:bool) {is=title K2Node_FunctionEntry}
   > Branch(Condition=@n1)
   > n2 = Set WarningWidget(WarningWidget=GetWarningWidget())
   > Is Valid(InputObject=@n2)
   > RemoveWarning(self=Get WarningWidget(), Type=NewEnumerator0)
   > RemoveWarning(self=Get WarningWidget(), Type=NewEnumerator1)
   > RemoveWarning(self=Get WarningWidget(), Type=NewEnumerator6)
   > RemoveWarning(self=Get WarningWidget(), Type=NewEnumerator7)
  ]}
]
variables [
 bIsVehicleElectric {type=boolean}
 DisabledTag {type=GameplayTag, default {tagName=Vehicle.Disabled}}
 LowFuelWarningActive {type=boolean}
 OutOfFuelWarningActive {type=boolean}
 RefuelingWarningActive {type=boolean}
]
widgets {
 info {inheritedWidgetCount=0, parentClass=/Script/FortniteUI.FortAthenaVehicleDashboardWidget, rootWidgetClass=null, widgetCount=0}
 widgets []
}
