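# Vehicle dashboard HUD widget that shows one status warning at a time for fuel, battery, charging, fueling or disabled states.
# ShowWarning inserts types from the WarningData map into a priority-sorted active list; RemoveWarning takes them out.
# UpdateShowPriority displays the top entry's text, colors, icon and optional bolt/pulse, collapsing the widget when the list empties.
# The Intro animation plays only when nothing was previously displayed.
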WidgetBlueprint /Game/Athena/HUD/Vehicles/Dashboard/WBP_Vehicle_Warning
animations [Intro {
  binding_count=3
  length_seconds=0.6074
  object=&self:Intro
  sequence {bindings [
    Overlay_Warning {
     id=#1
     possessed_class=Overlay
     tracks [{
       type=MovieScene2DTransformTrack
       sections [{
         type=MovieScene2DTransformSection
         channels [
          Translation.X {type=MovieSceneScriptingFloatChannel}
          Translation.Y {type=MovieSceneScriptingFloatChannel}
          Angle {type=MovieSceneScriptingFloatChannel}
          Scale.X {type=MovieSceneScriptingFloatChannel, keys [[0,0],[5.1863,0.9605],[5.4153,1]]}
          Scale.Y {type=MovieSceneScriptingFloatChannel, keys [[0,0],[5.1983,0.9654],[5.4153,1]]}
          Shear.X {type=MovieSceneScriptingFloatChannel}
          Shear.Y {type=MovieSceneScriptingFloatChannel}
         ]
        }]
      }]
    }
    Image_RoundCard_Bolt {
     id=#2
     possessed_class=Image
     tracks [{
       type=MovieScene2DTransformTrack
       sections [{
         type=MovieScene2DTransformSection
         channels [
          Translation.X {type=MovieSceneScriptingFloatChannel}
          Translation.Y {type=MovieSceneScriptingFloatChannel}
          Angle {type=MovieSceneScriptingFloatChannel}
          Scale.X {type=MovieSceneScriptingFloatChannel, keys [[4.2267,0],[8.2807,1]]}
          Scale.Y {type=MovieSceneScriptingFloatChannel, keys [[4.2267,0],[8.2807,1]]}
          Shear.X {type=MovieSceneScriptingFloatChannel}
          Shear.Y {type=MovieSceneScriptingFloatChannel}
         ]
        }]
      }]
    }
    Text_Warning {
     id=#3
     possessed_class=CommonTextBlock
     tracks [
      {
       type=MovieSceneFloatTrack
       display_name=Render Opacity
       sections [{type=MovieSceneFloatSection, channels [None {type=MovieSceneScriptingFloatChannel, keys [[6.7033,0],[9.158,1]]}]}]
      }
      {
       type=MovieScene2DTransformTrack
       sections [{
         type=MovieScene2DTransformSection
         channels [
          Translation.X {type=MovieSceneScriptingFloatChannel, keys [[6.7033,-231.2176],[12.148,0]]}
          Translation.Y {type=MovieSceneScriptingFloatChannel}
          Angle {type=MovieSceneScriptingFloatChannel}
          Scale.X {type=MovieSceneScriptingFloatChannel}
          Scale.Y {type=MovieSceneScriptingFloatChannel}
          Shear.X {type=MovieSceneScriptingFloatChannel}
          Shear.Y {type=MovieSceneScriptingFloatChannel}
         ]
        }]
      }
     ]
    }
   ]}
 }]
graphs [
 EventGraph {nodes [
   ShowLowFuelWarning() {is=title K2Node_CustomEvent}
   > ShowWarning!(execute, self, Type:ENUM_WarningType Enum=NewEnumerator0) -> (then)
   ShowOutOfFuelWarning() {is=title K2Node_CustomEvent}
   > ShowWarning(Type=NewEnumerator1)
  ]}
 RemoveWarning {nodes [
   n1 = RemoveWarning!() -> (then, Type:ENUM_WarningType Enum=NewEnumerator3) {is=title K2Node_FunctionEntry}
   > n2 = Remove Item!(
    execute
    TargetArray:Array of ENUM_WarningType Enums=Get ActiveWarningTypes() -> (ActiveWarningTypes:Array of ENUM_WarningType Enums=NewEnumerator3)
    Item:"ENUM_WarningType Enum (by ref)"=@n1
   ) -> (then, ReturnValue:bool)
   > Branch(Condition=@n2)
   > UpdateShowPriority()
  ]}
 ShowWarning {nodes [
   n1 = ShowWarning!() -> (then, Type:ENUM_WarningType Enum=NewEnumerator3) {is=title K2Node_FunctionEntry}
   n2 = Find(
    TargetMap:Map of ENUM_WarningType Enums to STRUCT Warning Data Structures=Get WarningData() -> (WarningData:Map of ENUM_WarningType Enums to STRUCT Warning Data Structures=NewEnumerator3)
    Key:"ENUM_WarningType Enum (by ref)"=@n1
   ) -> (Value:STRUCT Warning Data Structure)
   > Set WarningData(WarningData:STRUCT Warning Data Structure=@n2) -> (Output_Get:STRUCT Warning Data Structure)
   > Branch(Condition=@n2.ReturnValue)
   > n3 = Branch(Condition=Contains Item(
     TargetArray:Array of ENUM_WarningType Enums=Get ActiveWarningTypes() -> (ActiveWarningTypes=NewEnumerator3)
     ItemToFind:"ENUM_WarningType Enum (by ref)"=@n1
    ))
   Set InsertAtIndex(execute=@n3.else, InsertAtIndex:int=0) -> (Output_Get:int=0)
   > n4 = For Each Loop with Break!(Exec:Exec, Array:Array of ENUM_WarningType Enums=Get ActiveWarningTypes() -> (ActiveWarningTypes=NewEnumerator3), Break:Exec=@n6.then) -> (LoopBody:Exec, Array Element:ENUM_WarningType Enum, Array Index:int, Completed:Exec)
   Insert!(
    execute=@n4.Completed
    TargetArray:Array of ENUM_WarningType Enums=Get ActiveWarningTypes() -> (ActiveWarningTypes=NewEnumerator3)
    NewItem:"ENUM_WarningType Enum (by ref)"=Get Type() -> (Type:ENUM_WarningType Enum=NewEnumerator3)
    Index:int=Get InsertAtIndex() -> (InsertAtIndex:int=0)
   ) -> (then) {is=title K2Node_CallArrayFunction}
   > UpdateShowPriority()
   n5 = Find(TargetMap=Get WarningData() -> (WarningData=NewEnumerator3), Key=@n4)
   n6 = Branch(
    execute=@n4.LoopBody
    Condition="integer < integer"(
     A=Break STRUCT Warning Data!(STRUCT_WarningData:"STRUCT Warning Data Structure (by ref)"=@n5) -> (
      WarningText_2_#4:Text
      Image_5_#5:Material Instance Object Reference
      Priority_8_#6:int
      ShowBolt_11_#7:bool=~
      TextShadowColor_15_#8:Linear Color Structure
      TextOutlineColor_18_#9:Linear Color Structure
     ) {is=title K2Node_BreakStruct}.Priority_8_#6
     B=Get!() -> (
      WarningData_WarningText_2_#4:Text
      WarningData_Image_5_#5:Material Instance Object Reference
      WarningData_Priority_8_#6:int
      WarningData_ShowBolt_11_#7:bool=~
      WarningData_TextShadowColor_15_#8:Linear Color Structure
      WarningData_TextOutlineColor_18_#9:Linear Color Structure
     ).WarningData_Priority_8_#6
    )
   )
   Increment Int(" "=@n6.else, Value=Get InsertAtIndex() -> (InsertAtIndex=0))
  ]}
 UpdateShowPriority {nodes [
   UpdateShowPriority() {is=title K2Node_FunctionEntry}
   n1 = Get ActiveWarningTypes() -> (ActiveWarningTypes=NewEnumerator3)
   > n2 = Branch(Condition="integer > integer"(A=Length(TargetArray:Array of ENUM_WarningType Enums=@n1)))
   > n3 = Set NewDisplayedType(NewDisplayedType:ENUM_WarningType Enum="Get (a ref)"(Array:Array of ENUM_WarningType Enums=Get ActiveWarningTypes() -> (ActiveWarningTypes=NewEnumerator3), Dimension 1=0) -> (Output:"ENUM_WarningType Enum (by ref)")) -> (Output_Get:ENUM_WarningType Enum=NewEnumerator3)
   > Branch(Condition="Not Equal (Enum)"!(
     A:ENUM_WarningType Enum=@n3
     B:ENUM_WarningType Enum=Get CurrentDisplayedType() -> (CurrentDisplayedType:ENUM_WarningType Enum=NewEnumerator3)
    ) -> (ReturnValue:bool))
   > Set NewDisplayedData(NewDisplayedData:STRUCT Warning Data Structure=Find(
     TargetMap=Get WarningData() -> (WarningData=NewEnumerator3)
     Key=Get NewDisplayedType() -> (NewDisplayedType:ENUM_WarningType Enum=NewEnumerator3)
    )) -> (Output_Get:STRUCT Warning Data Structure)
   n4 = Get!() -> (
    NewDisplayedData_WarningText_2_#4:Text
    NewDisplayedData_Image_5_#5:Material Instance Object Reference
    NewDisplayedData_Priority_8_#6:int
    NewDisplayedData_ShowBolt_11_#7:bool=~
    NewDisplayedData_TextShadowColor_15_#8:Linear Color Structure
    NewDisplayedData_TextOutlineColor_18_#9:Linear Color Structure
   )
   > "SetText (Text)"(self=Get Text_Warning() -> (Text_Warning:Common Text Object Reference), InText=@n4)
   > SetShadowColorAndOpacity!(
    execute
    self:Text Object Reference=Get Text_Warning()
    InShadowColorAndOpacity:Linear Color Structure=@n4.NewDisplayedData_TextShadowColor_15_#8
   ) -> (then)
   > Set Outline Color!(
    In:Exec
    CommonTextWidget:Common Text Object Reference=Get Text_Warning()
    Color:Linear Color Structure=@n4.NewDisplayedData_TextOutlineColor_18_#9
   ) -> (Out:Exec) {is=title K2Node_MacroInstance}
   > n5 = Is Valid(InputObject=@n4.NewDisplayedData_Image_5_#5)
   SetVisibility(
    execute=[@n7.then, @n5.Is Not Valid]
    self=[
     Get Image_WarningPulse() -> (Image_WarningPulse:Image Object Reference)
     Get Image_RoundCard_Bolt() -> (Image_RoundCard_Bolt:Image Object Reference)
    ]
    InVisibility=Select(Option 0=Collapsed, Option 1=SelfHitTestInvisible, Index=@n4.NewDisplayedData_ShowBolt_11_#7)
   )
   > n6 = Sequence()
   > Branch(Condition="Equal (Enum)"(A:ENUM_WarningType Enum=Get CurrentDisplayedType() -> (CurrentDisplayedType=NewEnumerator3), B:ENUM_WarningType Enum=NewEnumerator3))
   > SetVisibility(InVisibility=SelfHitTestInvisible)
   > PlayAnimation(InAnimation=Get Intro(), bRestoreState=true)
   Set CurrentDisplayedType(execute=@n6.then_1, CurrentDisplayedType:ENUM_WarningType Enum=Get NewDisplayedType() -> (NewDisplayedType=NewEnumerator3)) -> (Output_Get:ENUM_WarningType Enum=NewEnumerator3)
   n7 = SetBrushFromMaterial(
    execute=@n5.Is Valid
    self=Get Image_WarningIcon() -> (Image_WarningIcon:Image Object Reference)
    Material=@n4.NewDisplayedData_Image_5_#5
   )
   Set CurrentDisplayedType(execute=@n2.else, CurrentDisplayedType=NewEnumerator3) -> (Output_Get=NewEnumerator3)
   > SetVisibility(InVisibility=Collapsed)
  ]}
]
variables [
 ActiveWarningTypes {type="array<ENUM_WarningType>", default []}
 CurrentDisplayedType {type=ENUM_WarningType, default=None}
 WarningData {default {
   Charging {
    image=null
    priority=2
    showBolt=false
    warningText=Charging...
    textOutlineColor=<rgba 0.0591 0.1719 0 1>
    textShadowColor=<rgba 0.1435 1 0 1>
   }
   Disabled {
    image=/Game/Athena/HUD/Vehicles/Materials/M_UI_LowFuel_Disabled
    priority=5
    showBolt=false
    warningText=Disabled
    textOutlineColor=<rgba 0 0.2043 1 1>
    textShadowColor=<rgba 0 0.0132 0.0625 1>
   }
   Fueling {
    image=/Game/Athena/HUD/Vehicles/Materials/MI_UI_RefuellingIcon
    priority=1
    showBolt=false
    warningText=Fueling...
    textOutlineColor=<rgba 0.1704 0.5694 0.0958 1>
    textShadowColor=<rgba 0.0467 0.1559 0.0262 1>
   }
   LowBattery {
    image=/Game/Athena/HUD/Vehicles/Materials/M_UI_LowBattery_Inst
    priority=3
    showBolt=true
    warningText=Battery Low
    textOutlineColor=<rgba 1 0.2021 0 1>
    textShadowColor=<rgba 0.0625 0.0024 0 1>
   }
   LowFuel {
    image=/Game/Athena/HUD/Vehicles/Materials/M_UI_LowFuel_Inst
    priority=3
    showBolt=true
    warningText=Low Fuel
    textOutlineColor=<rgba 1 0.2021 0 1>
    textShadowColor=<rgba 0.0625 0.0024 0 1>
   }
   NoBattery {
    image=/Game/Athena/HUD/Vehicles/Materials/M_UI_EmptyBattery_Inst
    priority=4
    showBolt=true
    warningText=Battery Empty
    textOutlineColor=<rgba 1 0.2021 0 1>
    textShadowColor=<rgba 0.0625 0.0024 0 1>
   }
   NoFuel {
    image=/Game/Athena/HUD/Vehicles/Materials/M_UI_LowFuel_Empty
    priority=4
    showBolt=true
    warningText=No Fuel
    textOutlineColor=<rgba 1 0.2021 0 1>
    textShadowColor=<rgba 0.0625 0.0024 0 1>
   }
  }}
]
widgets {
 info {inheritedWidgetCount=0, parentClass=&UserWidget, rootWidgetClass=&Overlay, widgetCount=5}
 widgets [
  Overlay_Warning {widgetClassPath=&Overlay, widget=Overlay_Warning, properties {renderTransformPivot=<1 1>}}
  Image_WarningIcon {
   widgetClassPath=&Image
   bIsVariable=true
   parent=Overlay_Warning
   widget=Image_WarningIcon
   properties {brush {
     bIsDynamicallyLoaded=false
     drawAs=Image
     imageType=NoImage
     mirroring=NoMirror
     resourceName=None
     resourceObject=&M_UI_LowFuel_Inst
     tiling=NoTile
     imageSize=<200 200>
     margin=<ltrb 0 0 0 0>
     outlineSettings {
      bUseBrushTransparency=false
      roundingType=HalfHeightRadius
      width=0
      color {colorUseRule=UseColor_Specified, specifiedColor=<rgba 0 0 0 0>}
      cornerRadii=<0 0 0 0>
     }
     tintColor {colorUseRule=UseColor_Specified, specifiedColor=<rgba 1 1 1 1>}
     uVRegion {bIsValid=false, max=<0 0>, min=<0 0>}
    }}
   slot {class=OverlaySlot}
  }
  Image_WarningPulse {
   widgetClassPath=&Image
   bIsVariable=true
   parent=Overlay_Warning
   widget=Image_WarningPulse
   properties {brush {
     bIsDynamicallyLoaded=false
     drawAs=Image
     imageType=NoImage
     mirroring=NoMirror
     resourceName=None
     resourceObject=/Game/Athena/HUD/LTM/Wax/Materials/M_UI_TextGlow_Add_Gradient_Orange
     tiling=NoTile
     imageSize=<200 200>
     margin=<ltrb 0 0 0 0>
     outlineSettings {
      bUseBrushTransparency=false
      roundingType=HalfHeightRadius
      width=0
      color {colorUseRule=UseColor_Specified, specifiedColor=<rgba 0 0 0 0>}
      cornerRadii=<0 0 0 0>
     }
     tintColor {colorUseRule=UseColor_Specified, specifiedColor=<rgba 1 1 1 1>}
     uVRegion {bIsValid=false, max=<0 0>, min=<0 0>}
    }}
   slot {class=OverlaySlot}
  }
  Image_RoundCard_Bolt {
   widgetClassPath=&Image
   bIsVariable=true
   parent=Overlay_Warning
   widget=Image_RoundCard_Bolt
   properties {
    visibility=HitTestInvisible
    brush {
     bIsDynamicallyLoaded=false
     drawAs=Image
     imageType=NoImage
     mirroring=NoMirror
     resourceName=None
     resourceObject=/Game/UI/Frontend/Showdown/Materials/M_UI_SkewBoxBorder_Tournament_ModeAlertBorder_Tail
     tiling=NoTile
     imageSize=<16.953 75.5717>
     margin=<ltrb 0 0 0 0>
     outlineSettings {
      bUseBrushTransparency=false
      roundingType=HalfHeightRadius
      width=0
      color {colorUseRule=UseColor_Specified, specifiedColor=<rgba 0 0 0 0>}
      cornerRadii=<0 0 0 0>
     }
     tintColor {colorUseRule=UseColor_Specified, specifiedColor=<rgba 1 1 1 1>}
     uVRegion {bIsValid=false, max=<0 0>, min=<0 0>}
    }
    renderTransform {angle=-26, scale=<1 1>, shear=<33 0>, translation=<27 77>}
    renderTransformPivot=<0.5 0>
   }
   slot {class=OverlaySlot, properties {horizontalAlignment=HAlign_Center, verticalAlignment=VAlign_Center}}
  }
  Text_Warning {
   widgetClassPath=&CommonTextBlock
   bIsVariable=true
   parent=Overlay_Warning
   widget=Text_Warning
   properties {
    text="{Text}"
    textTransformPolicy=ToUpper
    font {
     bForceMonospaced=false
     bMaterialIsStencil=false
     fontMaterial=null
     fontObject=&Burbank
     letterSpacing=0
     monospacedWidth=1
     size=24
     skewAmount=0.25
     typefaceFontName=BurbankRegular-Black
     outlineSettings {
      bApplyOutlineToDropShadows=false
      bMiteredCorners=false
      bSeparateFillAlpha=false
      outlineMaterial=null
      outlineSize=1
      outlineColor=<rgba 1 0.2021 0 1>
     }
    }
    renderTransform {angle=0, scale=<1 1>, shear=<0 0>, translation=<142 0>}
    shadowColorAndOpacity=<rgba 0.0625 0.0024 0 1>
    shadowOffset=<2 2>
   }
   slot {class=OverlaySlot, properties {verticalAlignment=VAlign_Center}}
  }
 ]
}
