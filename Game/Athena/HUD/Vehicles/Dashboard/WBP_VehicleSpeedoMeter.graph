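# Vehicle HUD speedometer widget, extending WBP_VehicleGaugeBase with a speed dial, fuel gauge, low-fuel warning, boost meter and vehicle name.
# Parent events toggle the rechargeable boost meter, forward infinite fuel to the gauge, and animate the name in, then out after a timer.
# A visibility flag guards the boost spark animation, which reverses when boost ends or the widget hides.
# It starts collapsed, and the BoostFull animation is not wired to the boost-filled event.
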WidgetBlueprint /Game/Athena/HUD/Vehicles/Dashboard/WBP_VehicleSpeedoMeter
animations [
 NameIntro {
  binding_count=1
  length_seconds=0.35
  object=&self:NameIntro
  sequence {bindings [Text_ItemName {
     id=#1
     possessed_class=CommonTextBlock
     tracks [
      {
       type=MovieSceneFloatTrack
       display_name=Render Opacity
       sections [{type=MovieSceneFloatSection, end_frame=7, start_frame=0, channels [None {type=MovieSceneScriptingFloatChannel, keys [[1.5157,0],[7,1]]}]}]
      }
      {
       type=MovieScene2DTransformTrack
       sections [{
         type=MovieScene2DTransformSection
         end_frame=7
         start_frame=0
         channels [Translation.X {type=MovieSceneScriptingFloatChannel, keys [[0,63.2905],[7,0]]}]
        }]
      }
     ]
    }]}
 }
 NameOutro {
  binding_count=1
  length_seconds=1.0019
  object=&self:NameOutro
  sequence {bindings [Text_ItemName {
     id=#2
     possessed_class=CommonTextBlock
     tracks [
      {
       type=MovieSceneFloatTrack
       display_name=Render Opacity
       sections [{type=MovieSceneFloatSection, channels [None {type=MovieSceneScriptingFloatChannel, keys [[0,1],[20.037,0]]}]}]
      }
      {
       type=MovieScene2DTransformTrack
       sections [{type=MovieScene2DTransformSection, channels [Translation.X {type=MovieSceneScriptingFloatChannel, keys [[0,0],[20.037,-20]]}]}]
      }
     ]
    }]}
 }
 BoostEngage {
  binding_count=2
  length_seconds=0.25
  object=&self:BoostEngage
  sequence {bindings [
    Image_BoostMeter {
     id=#3
     possessed_class=Image
     tracks [{
       type=MovieSceneWidgetMaterialTrack
       sections [{type=MovieSceneParameterSection, channels [ColorOverrideScalar {type=MovieSceneScriptingFloatChannel, keys [[0,0],[5,1]]}]}]
      }]
    }
    Overlay_BoostSparks {
     id=#4
     possessed_class=Overlay
     tracks [{
       type=MovieSceneFloatTrack
       display_name=Render Opacity
       sections [{type=MovieSceneFloatSection, end_frame=5, start_frame=0, channels [None {type=MovieSceneScriptingFloatChannel, keys [[0,0],[5,1]]}]}]
      }]
    }
   ]}
 }
 BoostFull {
  binding_count=1
  length_seconds=4.4454
  object=&self:BoostFull
  sequence {bindings [Image_BoostIcon {
     id=#5
     possessed_class=Image
     tracks [{
       type=MovieSceneWidgetMaterialTrack
       sections [{
         type=MovieSceneParameterSection
         channels [
          "TimeOffset (S)" {type=MovieSceneScriptingFloatChannel, keys [[0,0],[21.7223,1],[88.9073,1]]}
          IconScale {type=MovieSceneScriptingFloatChannel, keys [[0,0.2476],[4.1223,0.45],[21.7223,0.2476]]}
         ]
        }]
      }]
    }]}
 }
]
graphs [
 EventGraph {nodes [
   StopBoostStuff() {is=title K2Node_CustomEvent}
   > Branch(Condition=Get BoostEffectsVisible_0() -> (BoostEffectsVisible_0:bool))
   > PlayAnimationReverse(InAnimation=Get BoostEngage() -> (BoostEngage:Widget Animation Object Reference), PlaybackSpeed=0.5)
   > SetVisibility(self=Get Overlay_BoostSparks() -> (Overlay_BoostSparks:Overlay Object Reference), InVisibility=Hidden)
   > Set BoostEffectsVisible_0(BoostEffectsVisible_0:bool=false) -> (Output_Get:bool)
   n1 = OnPopUpOnSwitchDone() {is=title K2Node_CustomEvent}
   > StopAnimation(InAnimation=Get NameIntro() -> (NameIntro:Widget Animation Object Reference))
   > PlayAnimation(InAnimation=Get NameOutro() -> (NameOutro:Widget Animation Object Reference))
   n2 = Event OnWidgetVisibilityChanged!() -> (OutputDelegate, then, bNewVisibility:bool) {is=title K2Node_Event}
   > n3 = Branch(Condition=@n2.bNewVisibility)
   StopBoostStuff(execute=@n3.else) {is=title K2Node_CallFunction}
   n4 = Event SetVehicleHasInfiniteFuel!() -> (OutputDelegate, then, bInfiniteFuel:bool) {is=title K2Node_Event}
   > Parent: SetVehicleHasInfiniteFuel!(execute, bInfiniteFuel:bool=@n4.bInfiniteFuel) -> (then) {is=title K2Node_CallParentFunction}
   > ShowInfiniteFuel!(
    execute
    self:Fort Vehicle Fuel Gauge Object Reference=Get Widget_FuelGauge() -> (Widget_FuelGauge:Fort Vehicle Fuel Gauge Object Reference)
    bHasInfiniteFuel:bool=@n4.bInfiniteFuel
   ) -> (then)
   > UpdateFuelChanged!(execute, self:Fort Vehicle Fuel Gauge Object Reference=Get Widget_FuelGauge(), FuelPercent:float=1, bIsFuelLow:bool=false) -> (then)
   Event OnBoostStarted() {is=title K2Node_Event}
   > n5 = Branch(Condition=Get BoostEffectsVisible_0())
   PlayAnimationForward(execute=@n5.else, InAnimation=Get BoostEngage(), PlaybackSpeed=4)
   > SetVisibility(self=Get Overlay_BoostSparks(), InVisibility=HitTestInvisible)
   > Set BoostEffectsVisible_0(BoostEffectsVisible_0=true)
   Event OnBoostFinished() {is=title K2Node_Event}
   > StopBoostStuff()
   n6 = Event SetVehicleUsesRechargableBoost!() -> (OutputDelegate, then, bUsesRechargeableBoost:bool) {is=title K2Node_Event}
   > SetVisibility(
    self=[Get Spacer_Boost() -> (Spacer_Boost:Spacer Object Reference), Get OverlayBoostMeter() -> (OverlayBoostMeter:Overlay Object Reference)]
    InVisibility=Select(Option 0=Collapsed, Option 1=HitTestInvisible, Index=@n6.bUsesRechargeableBoost)
   )
   Event Construct()
   > SetVisibility(InVisibility=Collapsed)
   Event OnPlayerFacingNameUpdated() {is=title K2Node_Event}
   > Set Timer by Event(Delegate=@n1, Time=2, bLooping=false)
   > StopAnimation(InAnimation=Get NameOutro())
   > PlayAnimation(InAnimation=Get NameIntro())
   PlayAnimationForward(InAnimation=Get BoostFull() -> (BoostFull:Widget Animation Object Reference))
   Event PreConstruct()
   Event OnRechargableBoostFilled() {is=title K2Node_Event}
  ]}
 GetWarningWidget {nodes [
   GetWarningWidget() {is=title K2Node_FunctionEntry}
   > Return Node!(
    execute
    WarningWidget:WBP Vehicle Warning Object Reference=Get Widget_LowFuelWarning() -> (Widget_LowFuelWarning:WBP Vehicle Warning Object Reference)
   )
  ]}
]
variables [BoostEffectsVisible_0 {type=boolean}]
widgets {
 info {
  inheritedWidgetCount=10
  parentClass=&/WBP_VehicleGaugeBase.WBP_VehicleGaugeBase_C
  rootWidgetClass=&Overlay
  widgetCount=25
 }
 widgets [
  Overlay_Gauge {widgetClassPath=&Overlay, bInherited=true, widget=Overlay_Gauge, properties {visibility=HitTestInvisible}}
  Widget_LowFuelWarning {
   widgetClassPath=&/WBP_Vehicle_Warning
   bIsVariable=true
   parent=Overlay_Gauge
   widget=Widget_LowFuelWarning
   properties {visibility=Collapsed, renderTransform {angle=0, scale=<1 1>, shear=<0 0>, translation=<-428 -183>}}
   slot {class=OverlaySlot, properties {horizontalAlignment=HAlign_Right, verticalAlignment=VAlign_Bottom, padding=<ltrb 0 0 10 0>}}
  }
  HorizontalBox_FuelCluster {
   widgetClassPath=&HorizontalBox
   bInherited=true
   parent=Overlay_Gauge
   widget=HorizontalBox_FuelCluster
   properties {renderTransform {angle=0, scale=<1 1>, shear=<0 0>, translation=<-242 -52>}}
   slot {class=OverlaySlot, properties {horizontalAlignment=HAlign_Right, verticalAlignment=VAlign_Bottom, padding=<ltrb 0 0 10 0>}}
  }
  Spacer_Boost {
   widgetClassPath=&Spacer
   bIsVariable=true
   parent=HorizontalBox_FuelCluster
   widget=Spacer_Boost
   properties {visibility=Collapsed, size=<65 1>}
   slot {class=HorizontalBoxSlot, properties {horizontalAlignment=HAlign_Right}}
  }
  Widget_FuelGauge {
   widgetClassPath=&/WBP_Vehicle_FuelGauge
   bInherited=true
   bIsVariable=true
   parent=HorizontalBox_FuelCluster
   widget=Widget_FuelGauge
   slot {class=HorizontalBoxSlot, properties {horizontalAlignment=HAlign_Right, verticalAlignment=VAlign_Center}}
  }
  Overlay_SpeedCluster {
   widgetClassPath=&Overlay
   parent=Overlay_Gauge
   widget=Overlay_SpeedCluster
   properties {renderTransform {angle=0, scale=<1 1>, shear=<0 0>, translation=<14 16>}}
   slot {class=OverlaySlot, properties {horizontalAlignment=HAlign_Right, verticalAlignment=VAlign_Bottom, padding=<ltrb 0 0 45 0>}}
  }
  OverlayBoostMeter {
   widgetClassPath=&Overlay
   bIsVariable=true
   parent=Overlay_SpeedCluster
   widget=OverlayBoostMeter
   properties {visibility=Collapsed, renderTransform {angle=0, scale=<-1 1>, shear=<0 0>, translation=<32 -26>}}
   slot {class=OverlaySlot, properties {horizontalAlignment=HAlign_Center, verticalAlignment=VAlign_Center, padding=<ltrb -200 -200 -200 -200>}}
  }
  Image_BoostMeter {
   widgetClassPath=&Image
   bInherited=true
   bIsVariable=true
   parent=OverlayBoostMeter
   widget=Image_BoostMeter
   properties {brush {
     bIsDynamicallyLoaded=false
     drawAs=Image
     imageType=FullColor
     mirroring=NoMirror
     resourceName=None
     resourceObject=/Game/Athena/HUD/Vehicles/Materials/MI_UI_WhiplashBoostMeter
     tiling=NoTile
     imageSize=<400 380>
     margin=<ltrb 0 0 0 0>
     outlineSettings {
      bUseBrushTransparency=false
      roundingType=HalfHeightRadius
      width=0
      color {colorUseRule=UseColor_Specified, specifiedColor=<rgba 0 0 0 0>}
      cornerRadii=<0 0 0 0>
     }
     tintColor {colorUseRule=UseColor_Specified, specifiedColor=<rgba 1 1 1 1>}
     uVRegion {bIsValid=false, max=<0 0>, min=<0 0>}
    }}
   slot {class=OverlaySlot, properties {padding=<ltrb 107 40 0 0>}}
  }
  Image_BoostIcon {
   widgetClassPath=&Image
   parent=OverlayBoostMeter
   widget=Image_BoostIcon
   properties {
    brush {
     bIsDynamicallyLoaded=false
     drawAs=Image
     imageType=FullColor
     mirroring=NoMirror
     resourceName=None
     resourceObject=/Game/Athena/HUD/Vehicles/Materials/M_UI_ValetBoostIcon
     tiling=NoTile
     imageSize=<30 30>
     margin=<ltrb 0 0 0 0>
     outlineSettings {
      bUseBrushTransparency=false
      roundingType=HalfHeightRadius
      width=0
      color {colorUseRule=UseColor_Specified, specifiedColor=<rgba 0 0 0 0>}
      cornerRadii=<0 0 0 0>
     }
     tintColor {colorUseRule=UseColor_Specified, specifiedColor=<rgba 1 1 1 1>}
     uVRegion {bIsValid=false, max=<0 0>, min=<0 0>}
    }
    renderTransform {angle=0, scale=<1 1>, shear=<0 0>, translation=<147 -96>}
   }
   slot {class=OverlaySlot, properties {verticalAlignment=VAlign_Center}}
  }
  Overlay_BoostSparks {
   widgetClassPath=&Overlay
   bInherited=true
   bIsVariable=true
   parent=OverlayBoostMeter
   widget=Overlay_BoostSparks
   properties {visibility=Collapsed, renderTransform {angle=28, scale=<1.25 1.45>, shear=<0 0>, translation=<-103 44>}, renderTransformPivot=<135 0.5>}
   slot {class=OverlaySlot, properties {horizontalAlignment=HAlign_Center, verticalAlignment=VAlign_Center}}
  }
  Overlay_47 {
   widgetClassPath=&Overlay
   parent=Overlay_BoostSparks
   widget=Overlay_47
   properties {visibility=HitTestInvisible, renderTransform {angle=-180, scale=<1 1>, shear=<0 0>, translation=<0 0>}}
   slot {class=OverlaySlot}
  }
  Spacer_42 {widgetClassPath=&Spacer, parent=Overlay_47, widget=Spacer_42, properties {visibility=HitTestInvisible}, slot {class=OverlaySlot}}
  Image_Pulse {
   widgetClassPath=&Image
   bIsVariable=true
   parent=Overlay_47
   widget=Image_Pulse
   properties {
    visibility=HitTestInvisible
    brush {
     bIsDynamicallyLoaded=false
     drawAs=Image
     imageType=NoImage
     mirroring=NoMirror
     resourceName=None
     resourceObject=/Game/Athena/HUD/LTM/Wax/Materials/M_UI_TextGlow_Pulse
     tiling=NoTile
     imageSize=<44.4227 97.4638>
     margin=<ltrb 0 0 0 0>
     outlineSettings {
      bUseBrushTransparency=false
      roundingType=HalfHeightRadius
      width=0
      color {colorUseRule=UseColor_Specified, specifiedColor=<rgba 0 0 0 0>}
      cornerRadii=<0 0 0 0>
     }
     tintColor {colorUseRule=UseColor_Specified, specifiedColor=<rgba 0.8906 0.1088 0 1>}
     uVRegion {bIsValid=false, max=<0 0>, min=<0 0>}
    }
    renderTransform {angle=0, scale=<1 1>, shear=<0 0>, translation=<0 8>}
   }
   slot {class=OverlaySlot, properties {horizontalAlignment=HAlign_Center, verticalAlignment=VAlign_Center, padding=<ltrb -200 -200 -200 -200>}}
  }
  Image_Glow {
   widgetClassPath=&Image
   bIsVariable=true
   parent=Overlay_47
   widget=Image_Glow
   properties {
    visibility=HitTestInvisible
    brush {
     bIsDynamicallyLoaded=false
     drawAs=Image
     imageType=NoImage
     mirroring=NoMirror
     resourceName=None
     resourceObject=/Game/Athena/HUD/Vehicles/Materials/M_UI_SparksGlow
     tiling=NoTile
     imageSize=<130 109.0002>
     margin=<ltrb 0 0 0 0>
     outlineSettings {
      bUseBrushTransparency=false
      roundingType=HalfHeightRadius
      width=0
      color {colorUseRule=UseColor_Specified, specifiedColor=<rgba 0 0 0 0>}
      cornerRadii=<0 0 0 0>
     }
     tintColor {colorUseRule=UseColor_Specified, specifiedColor=<rgba 1 1 1 1>}
     uVRegion {bIsValid=false, max=<0 0>, min=<0 0>}
    }
    renderTransform {angle=0, scale=<1 1>, shear=<0 0>, translation=<0 17>}
   }
   slot {class=OverlaySlot, properties {horizontalAlignment=HAlign_Center, verticalAlignment=VAlign_Center, padding=<ltrb -200 -200 -200 -200>}}
  }
  Image_Sparks {
   widgetClassPath=&Image
   bIsVariable=true
   parent=Overlay_47
   widget=Image_Sparks
   properties {
    visibility=HitTestInvisible
    brush {
     bIsDynamicallyLoaded=false
     drawAs=Image
     imageType=NoImage
     mirroring=NoMirror
     resourceName=None
     resourceObject=/Game/Athena/UI/InGame/Interrogation/Materials/M_UI_Starburst_Sparks
     tiling=NoTile
     imageSize=<295.3884 400>
     margin=<ltrb 0 0 0 0>
     outlineSettings {
      bUseBrushTransparency=false
      roundingType=HalfHeightRadius
      width=0
      color {colorUseRule=UseColor_Specified, specifiedColor=<rgba 0 0 0 0>}
      cornerRadii=<0 0 0 0>
     }
     tintColor {colorUseRule=UseColor_Specified, specifiedColor=<rgba 1 0.3682 0.0743 1>}
     uVRegion {bIsValid=false, max=<0 0>, min=<0 0>}
    }
   }
   slot {class=OverlaySlot, properties {horizontalAlignment=HAlign_Center, verticalAlignment=VAlign_Center, padding=<ltrb -200 -200 -200 -200>}}
  }
  Image_Sparks_1 {
   widgetClassPath=&Image
   bIsVariable=true
   parent=Overlay_47
   widget=Image_Sparks_1
   properties {
    visibility=HitTestInvisible
    brush {
     bIsDynamicallyLoaded=false
     drawAs=Image
     imageType=NoImage
     mirroring=NoMirror
     resourceName=None
     resourceObject=/Game/Athena/UI/InGame/Interrogation/Materials/M_UI_Starburst_Sparks_Inst
     tiling=NoTile
     imageSize=<120 400>
     margin=<ltrb 0 0 0 0>
     outlineSettings {
      bUseBrushTransparency=false
      roundingType=HalfHeightRadius
      width=0
      color {colorUseRule=UseColor_Specified, specifiedColor=<rgba 0 0 0 0>}
      cornerRadii=<0 0 0 0>
     }
     tintColor {colorUseRule=UseColor_Specified, specifiedColor=<rgba 1 0.7183 0.651 1>}
     uVRegion {bIsValid=false, max=<0 0>, min=<0 0>}
    }
   }
   slot {class=OverlaySlot, properties {horizontalAlignment=HAlign_Center, verticalAlignment=VAlign_Center, padding=<ltrb -200 -200 -200 -200>}}
  }
  Image_LargeCone {
   widgetClassPath=&Image
   bIsVariable=true
   parent=Overlay_47
   widget=Image_LargeCone
   properties {
    visibility=HitTestInvisible
    brush {
     bIsDynamicallyLoaded=false
     drawAs=Image
     imageType=NoImage
     mirroring=NoMirror
     resourceName=None
     resourceObject=/Game/Athena/HUD/Vehicles/Materials/M_UI_SparksGlow_Inst
     tiling=NoTile
     imageSize=<150 219.5743>
     margin=<ltrb 0 0 0 0>
     outlineSettings {
      bUseBrushTransparency=false
      roundingType=HalfHeightRadius
      width=0
      color {colorUseRule=UseColor_Specified, specifiedColor=<rgba 0 0 0 0>}
      cornerRadii=<0 0 0 0>
     }
     tintColor {colorUseRule=UseColor_Specified, specifiedColor=<rgba 1 1 1 1>}
     uVRegion {bIsValid=false, max=<0 0>, min=<0 0>}
    }
    renderTransform {angle=0, scale=<1 1>, shear=<0 0>, translation=<0 44>}
   }
   slot {class=OverlaySlot, properties {horizontalAlignment=HAlign_Center, verticalAlignment=VAlign_Center, padding=<ltrb -200 -200 -200 -200>}}
  }
  Overlay_Speedometer {widgetClassPath=&Overlay, parent=Overlay_SpeedCluster, widget=Overlay_Speedometer, slot {class=OverlaySlot}}
  Image_Speedometer {
   widgetClassPath=&Image
   bInherited=true
   bIsVariable=true
   parent=Overlay_Speedometer
   widget=Image_Speedometer
   properties {brush {
     bIsDynamicallyLoaded=false
     drawAs=Image
     imageType=NoImage
     mirroring=NoMirror
     resourceName=None
     resourceObject=/Game/Athena/HUD/Vehicles/Materials/M_UI_ValetDash_Speedometer_Inst
     tiling=NoTile
     imageSize=<330 330>
     margin=<ltrb 0 0 0 0>
     outlineSettings {
      bUseBrushTransparency=false
      roundingType=HalfHeightRadius
      width=0
      color {colorUseRule=UseColor_Specified, specifiedColor=<rgba 0 0 0 0>}
      cornerRadii=<0 0 0 0>
     }
     tintColor {colorUseRule=UseColor_Specified, specifiedColor=<rgba 1 1 1 1>}
     uVRegion {bIsValid=false, max=<0 0>, min=<0 0>}
    }}
   slot {class=OverlaySlot}
  }
  NumericText_Speed {
   widgetClassPath=/Script/CommonUI.CommonNumericTextBlock
   bInherited=true
   bIsVariable=true
   parent=Overlay_Speedometer
   widget=NumericText_Speed
   properties {
    bIsVolatile=true
    justification=Center
    performSizeInterpolation=false
    postInterpolationShrinkDuration=0
    text=0
    font {
     bForceMonospaced=false
     bMaterialIsStencil=false
     fontMaterial=null
     fontObject=&Burbank
     letterSpacing=0
     monospacedWidth=1
     size=52.5
     skewAmount=0
     typefaceFontName=BurbankRegular-Black
     outlineSettings {
      bApplyOutlineToDropShadows=false
      bMiteredCorners=false
      bSeparateFillAlpha=false
      outlineMaterial=null
      outlineSize=0
      outlineColor=<rgba 0 0 0 1>
     }
    }
    renderTransform {angle=0, scale=<1 1>, shear=<-8 0>, translation=<0 -5>}
    shadowOffset=<0 0>
   }
   slot {class=OverlaySlot, properties {horizontalAlignment=HAlign_Center, verticalAlignment=VAlign_Center}}
  }
  CommonTextBlock_72 {
   widgetClassPath=&CommonTextBlock
   parent=Overlay_Speedometer
   widget=CommonTextBlock_72
   properties {
    text=KPH
    visibility=Hidden
    font {
     bForceMonospaced=false
     bMaterialIsStencil=false
     fontMaterial=null
     fontObject=&Burbank
     letterSpacing=0
     monospacedWidth=1
     size=15
     skewAmount=0
     typefaceFontName=BurbankRegular-Black
     outlineSettings {
      bApplyOutlineToDropShadows=false
      bMiteredCorners=false
      bSeparateFillAlpha=false
      outlineMaterial=null
      outlineSize=0
      outlineColor=<rgba 0 0 0 1>
     }
    }
    renderTransform {angle=0, scale=<1 1>, shear=<-8 0>, translation=<-3 47>}
    renderTransformPivot=<0.5 0>
    shadowOffset=<0 0>
   }
   slot {class=OverlaySlot, properties {horizontalAlignment=HAlign_Center, verticalAlignment=VAlign_Center}}
  }
  Overlay_CarInfo {
   widgetClassPath=&Overlay
   parent=Overlay_Gauge
   widget=Overlay_CarInfo
   properties {renderTransform {angle=0, scale=<1 1>, shear=<0 0>, translation=<-402 -22>}}
   slot {class=OverlaySlot, properties {horizontalAlignment=HAlign_Right, verticalAlignment=VAlign_Bottom}}
  }
  Text_ItemName {
   widgetClassPath=&CommonTextBlock
   bInherited=true
   bIsVariable=true
   parent=Overlay_CarInfo
   widget=Text_ItemName
   properties {
    justification=Right
    text="[PH] Item Name"
    font {
     bForceMonospaced=false
     bMaterialIsStencil=false
     fontMaterial=null
     fontObject=&Burbank
     letterSpacing=0
     monospacedWidth=1
     size=20
     skewAmount=0
     typefaceFontName=Burbank-Black
     outlineSettings {
      bApplyOutlineToDropShadows=true
      bMiteredCorners=false
      bSeparateFillAlpha=true
      outlineMaterial=&M-FontShadow
      outlineSize=3
      outlineColor=<rgba 0 0 0 0.5>
     }
    }
    shadowOffset=<0 0>
   }
   slot {class=OverlaySlot}
  }
  Image_BoostGlow {widgetClassPath=&Image, bInherited=true, widget=null}
  TutorialHighlight {widgetClassPath=&TutorialHighlight}
 ]
}
