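# Mobile HUD element that hosts the default vehicle speedometer for the player's current vehicle.
# Player controller events for driver entry and switching to seat 0, plus Construct, cache the vehicle and check its custom UI for the gauge widget class.
# A match lazily creates one speedometer in an Overlay and binds it to the vehicle; lacking custom UI collapses visibility.
# Full exit unbinds it and clears the cached vehicle.
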WidgetBlueprint /Game/Athena/HUD/Vehicles/Dashboard/WBP_VehicleSpeedoMeter_Wrapper
graphs [
 EventGraph {nodes [
   OnEnterVehicleDriver() {is=title K2Node_CustomEvent}
   > n1 = Cast To FortPlayerController(Object=GetOwningPlayer())
   > n2 = Cast To FortPlayerPawnAthena(Object=GetPlayerPawn!(self:Fort Player Controller Object Reference=@n1) -> (ReturnValue:Fort Player Pawn Object Reference))
   n3 = Get Vehicle!(self:Fort Player Pawn Object Reference=@n2) -> (ReturnValue:Fort Athena Vehicle Object Reference) {is=title K2Node_CallFunction}
   > Is Valid(InputObject=@n3)
   > Set CachedVehicle(CachedVehicle:Fort Athena Vehicle Object Reference=@n3) -> (Output_Get:Fort Athena Vehicle Object Reference)
   > n4 = Branch(Condition=HasCustomUI!(self:Fort Vehicle Interface Interface=@n3) -> (ReturnValue:bool))
   > Branch(Condition=ContainsCustomUIDefinitionForClass!(
     self:Fort Vehicle Interface Interface=@n3
     WidgetClassToCheck:"User Widget Class Reference (by ref)"=Get Widget Class to Check() -> (Widget Class to Check:User Widget Class Reference)
    ) -> (ReturnValue:bool))
   > n5 = GetOrCreateDefaultSpeedometer!(execute, self) -> (then, DefaultSpeedometer:Fort Athena Vehicle Dashboard Widget Object Reference)
   > BindTo!(
    execute:Exec
    self:Object Reference=@n5
    AssociatedObject:Object Reference=Get CachedVehicle() -> (CachedVehicle:Fort Athena Vehicle Object Reference)
   ) -> (then:Exec) {is=title K2Node_Message}
   SetVisibility(execute=@n4.else, InVisibility=Collapsed)
   OnFullyExitVehicle() {is=title K2Node_CustomEvent}
   > n6 = Get!(execute:Exec) -> (then:Exec, else:Exec, CachedVehicle:Fort Athena Vehicle Object Reference)
   > n7 = Branch(Condition=HasCustomUI(self=@n6))
   n8 = Get DefaultSpeedometer() -> (DefaultSpeedometer:Fort Athena Vehicle Dashboard Widget Object Reference)
   > Is Valid(InputObject=@n8)
   > n9 = UnbindFrom!(execute:Exec, self:Object Reference=@n8, AssociatedObject:Object Reference=@n6) -> (then:Exec) {is=title K2Node_Message}
   Set CachedVehicle(execute=[@n7.else, @n9.then])
   n10 = OnSeatChanged!() -> (OutputDelegate, then, SeatIndex:int) {is=title K2Node_CustomEvent}
   > Branch(Condition="Equal (Integer)"(A=@n10.SeatIndex))
   > OnEnterVehicleDriver() {is=title K2Node_CallFunction}
   Event Construct()
   > n11 = Cast To FortPlayerController(Object=GetOwningPlayer())
   > n12 = Set ChachedPlayerController(ChachedPlayerController:Fort Player Controller Object Reference=@n11) -> (Output_Get:Fort Player Controller Object Reference)
   > Bind Event to On Enter Vehicle Driver(self:Fort Player Controller Object Reference=@n12, Delegate=Create Event()) {is=title K2Node_AddDelegate}
   > Bind Event to On Vehicle Seat Changed(
    self:Fort Player Controller Object Reference=Get ChachedPlayerController() -> (ChachedPlayerController:Fort Player Controller Object Reference)
    Delegate=Create Event()
   ) {is=title K2Node_AddDelegate}
   > Assign On Fully Exit Vehicle!(execute:Exec, self:Fort Player Controller Object Reference=Get ChachedPlayerController(), Delegate:"Delegate (by ref)"=Create Event()) -> (then:Exec) {is=title K2Node_AssignDelegate}
   > OnEnterVehicleDriver()
   Event Destruct()
   > n13 = Get!(execute) -> (then, else, ChachedPlayerController:Fort Player Controller Object Reference)
   > Unbind Event from On Enter Vehicle Driver(self:Fort Player Controller Object Reference=@n13, Delegate=Create Event()) {is=title K2Node_RemoveDelegate}
   > Unbind Event from On Fully Exit Vehicle(self:Fort Player Controller Object Reference=Get ChachedPlayerController(), Delegate=Create Event()) {is=title K2Node_RemoveDelegate}
   > Unbind Event from On Vehicle Seat Changed(self:Fort Player Controller Object Reference=Get ChachedPlayerController(), Delegate=Create Event()) {is=title K2Node_RemoveDelegate}
  ]}
 GetOrCreateDefaultSpeedometer {nodes [
   GetOrCreateDefaultSpeedometer() {is=title K2Node_FunctionEntry}
   > n1 = Is Valid(InputObject=Get DefaultSpeedometer())
   n2 = Create Widget!(
    execute:Exec=@n1.Is Not Valid
    Class:User Widget Class Reference=&/WBP_VehicleSpeedoMeter.WBP_VehicleSpeedoMeter_C
    OwningPlayer:Player Controller Object Reference=GetOwningPlayer()
   ) -> (then:Exec, ReturnValue:WBP Vehicle Speedo Meter Object Reference)
   > n3 = Set DefaultSpeedometer(DefaultSpeedometer:Fort Athena Vehicle Dashboard Widget Object Reference=@n2) -> (Output_Get:Fort Athena Vehicle Dashboard Widget Object Reference)
   > AddChild(self=Get Overlay() -> (Overlay:Overlay Object Reference), Content=@n3)
   > Return Node!(execute, DefaultSpeedometer:Fort Athena Vehicle Dashboard Widget Object Reference=@n3)
   Return Node(execute=@n1.Is Valid, DefaultSpeedometer=Get DefaultSpeedometer())
  ]}
]
variables [
 CachedVehicle {type=FortAthenaVehicle, default=null}
 ChachedPlayerController {type=FortPlayerController, default=null}
 DefaultSpeedometer {type=FortAthenaVehicleDashboardWidget, default=null}
 Widget Class to Check {type=Class, default=&/VehicleGaugeWidget.VehicleGaugeWidget_C}
]
widgets {
 info {inheritedWidgetCount=3, parentClass=/Script/FortniteUI.FortMobileHUDElement, rootWidgetClass=&ScaleBox, widgetCount=4}
 widgets [
  ScaleBox_HUDWidgetScale {widgetClassPath=&ScaleBox, bInherited=true, widget=ScaleBox_HUDWidgetScale, properties {stretch=UserSpecified}}
  Overlay {widgetClassPath=&Overlay, bIsVariable=true, parent=ScaleBox_HUDWidgetScale, widget=Overlay, slot {class=ScaleBoxSlot}}
  Image_Background {widgetClassPath=&Image, bInherited=true, widget=null}
  TutorialHighlight {widgetClassPath=&TutorialHighlight}
 ]
}
