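# Vehicle dashboard banner widget that flashes an out-of-fuel warning, reading "No Fuel!" or "Battery Empty!" for electric vehicles.
# ShowNoFuelMessage toggles the two texts by bIsElectric, then plays the NoFuelMessage animation forward with state restore.
# That animation briefly reveals the overlay with its darkened, glowing backdrop, punches the text scale, then fades out and collapses it.
# Reset stops the animation.
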WidgetBlueprint /Game/Athena/HUD/Vehicles/Dashboard/WBP_Vehicle_NoFuelBanner
animations [NoFuelMessage {
  binding_count=3
  length_seconds=2.7194
  object=&self:NoFuelMessage
  sequence {bindings [
    Text_NoCharge {
     id=#1
     possessed_class=CommonTextBlock
     tracks [{
       type=MovieScene2DTransformTrack
       sections [{
         type=MovieScene2DTransformSection
         channels [
          Translation.X {type=MovieSceneScriptingFloatChannel}
          Translation.Y {type=MovieSceneScriptingFloatChannel}
          Angle {type=MovieSceneScriptingFloatChannel}
          Scale.X {type=MovieSceneScriptingFloatChannel, keys [[3,1],[4.2057,2.7122],[6,2.9755],[8,1],[8.8983,1]]}
          Scale.Y {type=MovieSceneScriptingFloatChannel, keys [[3,1],[4.2057,2.7122],[6,2.9755],[8,1],[8.8983,1]]}
          Shear.X {type=MovieSceneScriptingFloatChannel}
          Shear.Y {type=MovieSceneScriptingFloatChannel}
         ]
        }]
      }]
    }
    Overlay_NoFuelMessage {
     id=#2
     possessed_class=Overlay
     tracks [
      {
       type=MovieSceneFloatTrack
       display_name=Render Opacity
       sections [{type=MovieSceneFloatSection, channels [None {type=MovieSceneScriptingFloatChannel, keys [[1,0],[3,1],[40.662,1],[54.388,0]]}]}]
      }
      {
       type=MovieSceneByteTrack
       display_name=Visibility
       sections [{type=MovieSceneByteSection, channels [None {type=MovieSceneScriptingByteChannel, keys [[0,1],[1,3],[54.388,1]]}]}]
      }
     ]
    }
    Text_NoFuel {
     id=#3
     possessed_class=CommonTextBlock
     tracks [{
       type=MovieScene2DTransformTrack
       sections [{
         type=MovieScene2DTransformSection
         channels [
          Translation.X {type=MovieSceneScriptingFloatChannel}
          Translation.Y {type=MovieSceneScriptingFloatChannel}
          Angle {type=MovieSceneScriptingFloatChannel}
          Scale.X {type=MovieSceneScriptingFloatChannel, keys [[3,1],[4.2057,2.7122],[6,2.9755],[8,1],[8.8983,1]]}
          Scale.Y {type=MovieSceneScriptingFloatChannel, keys [[3,1],[4.2057,2.7122],[6,2.9755],[8,1],[8.8983,1]]}
          Shear.X {type=MovieSceneScriptingFloatChannel}
          Shear.Y {type=MovieSceneScriptingFloatChannel}
         ]
        }]
      }]
    }
   ]}
 }]
graphs [EventGraph {nodes [
   n1 = ShowNoFuelMessage!() -> (OutputDelegate, then, bIsElectric:bool=~) {is=title K2Node_CustomEvent}
   > SetVisibility(
    self=Get Text_NoCharge() -> (Text_NoCharge:Common Text Object Reference)
    InVisibility=Select(Option 0=Collapsed, Option 1=SelfHitTestInvisible, Index=@n1.bIsElectric)
   )
   > SetVisibility(
    self=Get Text_NoFuel() -> (Text_NoFuel:Common Text Object Reference)
    InVisibility=Select(Option 0=SelfHitTestInvisible, Option 1=Collapsed, Index=@n1.bIsElectric)
   )
   n2 = Get NoFuelMessage() -> (NoFuelMessage:Widget Animation Object Reference)
   > PlayAnimationForward(InAnimation=@n2, bRestoreState=true)
   Reset() {is=title K2Node_CustomEvent}
   > StopAnimation(InAnimation=@n2)
  ]}]
variables []
widgets {
 info {inheritedWidgetCount=0, parentClass=&UserWidget, rootWidgetClass=&Overlay, widgetCount=6}
 widgets [
  Overlay_0 {widgetClassPath=&Overlay, widget=Overlay_0}
  Overlay_NoFuelMessage {
   widgetClassPath=&Overlay
   parent=Overlay_0
   widget=Overlay_NoFuelMessage
   properties {visibility=Collapsed, renderTransform {angle=0, scale=<1 1>, shear=<-12 0>, translation=<0 0>}}
   slot {class=OverlaySlot}
  }
  Image_darken {
   widgetClassPath=&Image
   parent=Overlay_NoFuelMessage
   widget=Image_darken
   properties {
    renderOpacity=0.5
    brush {
     bIsDynamicallyLoaded=false
     drawAs=Image
     imageType=NoImage
     mirroring=NoMirror
     resourceName=None
     resourceObject=/Game/Athena/HUD/LTM/Wax/Materials/M_UI_TextGlow
     tiling=NoTile
     imageSize=<32 32>
     margin=<ltrb 0 0 0 0>
     outlineSettings {
      bUseBrushTransparency=false
      roundingType=HalfHeightRadius
      width=0
      color {colorUseRule=UseColor_Specified, specifiedColor=<rgba 0 0 0 0>}
      cornerRadii=<0 0 0 0>
     }
     tintColor {colorUseRule=UseColor_Specified, specifiedColor=<rgba 0 0 0 1>}
     uVRegion {bIsValid=false, max=<0 0>, min=<0 0>}
    }
   }
   slot {class=OverlaySlot, properties {horizontalAlignment=HAlign_Fill, verticalAlignment=VAlign_Fill, padding=<ltrb 0 -50 0 -50>}}
  }
  Image_messageBG {
   widgetClassPath=&Image
   parent=Overlay_NoFuelMessage
   widget=Image_messageBG
   properties {
    visibility=HitTestInvisible
    brush {
     bIsDynamicallyLoaded=false
     drawAs=Image
     imageType=NoImage
     mirroring=NoMirror
     resourceName=None
     resourceObject=/Game/Athena/HUD/Vehicles/Materials/M_UI_FuelWarningMessageBG_Inst
     tiling=NoTile
     imageSize=<32 32>
     margin=<ltrb 0 0 0 0>
     outlineSettings {
      bUseBrushTransparency=false
      roundingType=HalfHeightRadius
      width=0
      color {colorUseRule=UseColor_Specified, specifiedColor=<rgba 0 0 0 0>}
      cornerRadii=<0 0 0 0>
     }
     tintColor {colorUseRule=UseColor_Specified, specifiedColor=<rgba 1 1 1 1>}
     uVRegion {bIsValid=false, max=<0 0>, min=<0 0>}
    }
   }
   slot {class=OverlaySlot, properties {horizontalAlignment=HAlign_Fill, verticalAlignment=VAlign_Fill}}
  }
  Text_NoFuel {
   widgetClassPath=&CommonTextBlock
   bIsVariable=true
   parent=Overlay_NoFuelMessage
   widget=Text_NoFuel
   properties {
    justification=Center
    text="No Fuel!"
    textTransformPolicy=ToUpper
    font {
     bForceMonospaced=false
     bMaterialIsStencil=false
     fontMaterial=null
     fontObject=&Burbank
     letterSpacing=0
     monospacedWidth=1
     size=24
     skewAmount=0
     typefaceFontName=BurbankRegular-Black
     outlineSettings {
      bApplyOutlineToDropShadows=true
      bMiteredCorners=false
      bSeparateFillAlpha=true
      outlineMaterial=/Game/UI/Foundation/Materials/M-FontShadow
      outlineSize=3
      outlineColor=<rgba 0 0 0 0.8>
     }
    }
    shadowOffset=<0 0>
   }
   slot {class=OverlaySlot, properties {padding=<ltrb 200 0 200 0>}}
  }
  Text_NoCharge {
   widgetClassPath=&CommonTextBlock
   bIsVariable=true
   parent=Overlay_NoFuelMessage
   widget=Text_NoCharge
   properties {
    justification=Center
    text="Battery Empty!"
    textTransformPolicy=ToUpper
    visibility=Collapsed
    font {
     bForceMonospaced=false
     bMaterialIsStencil=false
     fontMaterial=null
     fontObject=&Burbank
     letterSpacing=0
     monospacedWidth=1
     size=24
     skewAmount=0
     typefaceFontName=BurbankRegular-Black
     outlineSettings {
      bApplyOutlineToDropShadows=true
      bMiteredCorners=false
      bSeparateFillAlpha=true
      outlineMaterial=&M-FontShadow
      outlineSize=3
      outlineColor=<rgba 0 0 0 0.8>
     }
    }
    shadowOffset=<0 0>
   }
   slot {class=OverlaySlot, properties {padding=<ltrb 200 0 200 0>}}
  }
 ]
}
